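; Conversation task node "Back" in the FortniteConversation plugin, which sends a conversation back to the player's last choice.
; Its ExecuteTaskNode override calls ReturntoLastClientChoice on the conversation context and returns that result as the task's outcome.
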(fn ExecuteTaskNode
  (params (Context :type Conversation%20Context%20Structure%20%28by%20ref%29))
  (bind _returnvalue (Conversation|ReturntoLastClientChoice Context))
  (return _returnvalue))
