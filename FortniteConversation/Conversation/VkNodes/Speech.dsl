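; Conversation task node that shows an NPC's speech line, then pauses the conversation and sends the client its choices.
; It resolves a contextual message for the NPC participant, formats it with NPC and player names, and sends it with metadata.
; If resolution fails, usually because the text is blank, it still pauses and sends only the NPC's name.
; A helper looks up participant display names by tag, falling back to "???".
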(fn ExecuteTaskNode
  (params (Context :type Conversation%20Context%20Structure%20%28by%20ref%29))
  (bind _returnvalue (GameplayTags|MakeLiteralGameplayTag "(TagName=\"Conversation.Participant.NPC\")"))
  (bind (_outresult _returnvalue_2) (CallFunction|ResolveMessageforContext _returnvalue Context))
  (bind (_message _metadataparameters) (Utilities|Struct|BreakContextualMessageQueryResult _outresult))
  (if _returnvalue_2
    (bind _name (CallFunction|GetParticipantName (Variables|GetContext) "(TagName=\"Conversation.Participant.NPC\")"))
    (bind _name_2 (CallFunction|GetParticipantName (Variables|GetContext) "(TagName=\"Conversation.Participant.Player\")"))
    (bind _returnvalue_3 (Conversation|PauseConversationandSendClientChoices (Variables|GetContext) (Utilities|Struct|MakeClientConversationMessage "(TagName=\"\")" _name (Utilities|Text|FormatText _message _name _name_2) _metadataparameters)))
    (return _returnvalue_3)
    (else
      ;[1 1 1] If resolve message is false then the message was probably blank. This is not really an issue so continue conversation with no message.
      (bind _name_3 (CallFunction|GetParticipantName (Variables|GetContext) "(TagName=\"Conversation.Participant.NPC\")"))
      (bind _returnvalue_4 (Conversation|PauseConversationandSendClientChoices (Variables|GetContext) (Utilities|Struct|MakeClientConversationMessage "(TagName=\"\")" _name_3)))
      (return _returnvalue_4))))

(fn GetParticipantName
  (params (Context :type Conversation%20Context%20Structure) (ParticipantTag :type Gameplay%20Tag%20Structure))
  (bind _returnvalue (Conversation|GetConversationParticipant Context ParticipantTag))
  (Utilities|IsValid _returnvalue
    (:Is%20Valid
      (bind _returnvalue_2 (Conversation|GetParticipantDisplayName _returnvalue))
      (return _returnvalue_2))
    (:Is%20Not%20Valid
      (return "NSLOCTEXT(\"[921033C34ADBB83CDBA5D99368C18396]\", \"C64EEE644F4F30AC1E69A2892CD7795E\", \"???\")"))))
